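# Material function for the tertiary material layer of the Soccerball_Orville vehicle wheel cosmetic, computing stylized custom lighting from color, type, control vectors and diffuse mask.
# The type selects rows in a shared light-falloff lookup texture, indexed by view-angle and half-vector dot products.
# A tiled, strength-scaled normal perturbs specular and packed environment-map reflection terms.
# It outputs custom lighting, a diffuse lighting term and the tertiary normal.
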MaterialFunction /VehicleCosmetics/Wheels/Soccerball_Orville/Materials/MF_Tertiary
graph {nodes [
  n1 = Constant() {r=0.5}
  n2 = FunctionInput(Preview=Constant4Vector() {constant=<rgba 0.75 0.15 3 0.31>}.RGBA) {inputName=TetriaryMaterial_ControlA, inputType=FunctionInput_Vector4, sortPriority=1}
  n3 = ComponentMask(None=@n2) {g=true}
  n4 = Constant() {r=64}
  n5 = TextureObjectParameter() {parameterName=Param, samplerType=SAMPLERTYPE_Normal, texture=/VehicleCosmetics/Textures/UniversalTextures/T_CarbonFiber_Flipped_N}
  n6 = TextureSample(
   UVs=Multiply(
    A=TextureCoordinate()
    B=FunctionInput(Preview=@n4) {inputName=TetriaryMaterial_Normal_Tiling, inputType=FunctionInput_Scalar, sortPriority=1}
   )
   Tex=FunctionInput(Preview=@n5) {inputName=TetriaryMaterial_Normal, inputType=FunctionInput_Texture2D, sortPriority=1}
  )
  n7 = DotProduct(
   A=LinearInterpolate(A=Constant3Vector() {constant=<rgba 0 0 1 0>}, B=@n6.RGB)
   B=MaterialFunctionCall() {materialFunction=&CameraDirectionVector}.Result
  )
  n8 = Constant() {r=2}
  n9 = Add(
   A=Multiply(A=FunctionInput(Preview=@n8) {inputName=TetriaryMaterial_Type, inputType=FunctionInput_Scalar, sortPriority=1}, B=Constant() {r=0.0312})
   B=Constant() {r=0.0039}
  )
  n10 = Add(A=@n9, B=Constant() {r=0.0156})
  n11 = TextureSample(UVs=ComponentMask(None=AppendVector(A=@n7, B=@n10)) {g=true, r=true}) {texture=/VehicleCosmetics/Textures/UniversalTextures/T_LightFalloffArray}
  n12 = ConstantBiasScale(None=@n11.R) {bias=-0.5, scale=8}
  n13 = Constant4Vector() {constant=<rgba 0.1 0.1 0.1 1>}
  n14 = Multiply(
   A=ComponentMask(None=FunctionInput(Preview=@n13.RGBA) {inputName=TetriaryMaterial_Color, inputType=FunctionInput_Vector4, sortPriority=1}) {b=true, g=true, r=true}
   B=Clamp(None=Add(A=@n12) {constB=0.75})
  )
  n15 = Multiply(
   A=Multiply(A=LinearInterpolate(A=@n1, B=@n14, Alpha=@n3), B=@n11.G)
   B=FunctionInput(Preview=TextureSample() {samplerType=SAMPLERTYPE_LinearColor, texture=/VehicleCosmetics/Bodies/Aftershock/Textures/Body_Aftershock_D}.RGB) {inputName=Diffuse_A_Chanel, inputType=FunctionInput_Scalar, sortPriority=1}
  )
  n16 = DotProduct(A=@n6.A)
  n17 = ComponentMask(None=AppendVector(A=ConstantBiasScale(None=@n16), B=NamedRerouteUsage())) {g=true, r=true}
  n18 = Add(A=Multiply(A=Add(A=@n14, B=@n15), B=TextureSample(UVs=@n17) {texture=&T_LightFalloffArray}.R))
  n19 = LinearInterpolate(B=@n14, Alpha=ComponentMask(None=@n2) {r=true}) {constA=0.5}
  n20 = Add(
   A=LightVector()
   B=MaterialFunctionCall() {materialFunction=&CameraDirectionVector}.Result
  )
  n21 = ComponentMask(None=@n2) {a=true}
  n22 = Multiply(A=AppendVector(A=AppendVector(A=@n21, B=@n21), B=Constant() {r=1}), B=@n6.RGB)
  n23 = DotProduct(A=@n22, B=Normalize(VectorInput=Multiply(A=@n20) {constB=0.5}))
  n24 = Multiply(A=@n23, B=@n23)
  n25 = NamedRerouteUsage()
  n26 = TextureSample(UVs=ComponentMask(None=AppendVector(A=@n24, B=@n25)) {g=true, r=true}) {texture=&T_LightFalloffArray}
  n27 = FunctionInput(Preview=Constant4Vector() {constant=<rgba 0.75 0.15 3 0.31>}.RGBA) {inputName=TetriaryMaterial_ControlB, inputType=FunctionInput_Vector4, sortPriority=1}
  n28 = ComponentMask(None=@n27) {r=true}
  n29 = ConstantBiasScale(None=Transform(None=ReflectionVectorWS()))
  n30 = TextureSample(UVs=ComponentMask(None=Add(A=@n29, B=@n22)) {g=true, r=true}) {texture=/VehicleCosmetics/Textures/UniversalTextures/T_ENVPack}
  n31 = ComponentMask(None=AppendVector(A=LinearInterpolate(A=@n30.R, B=@n30.R), B=NamedRerouteUsage())) {g=true, r=true}
  n32 = ComponentMask(None=@n22) {b=true}
  n33 = Clamp(None=ConstantBiasScale(None=OneMinus(None=@n32)) {bias=0, scale=1})
  n34 = Add(
   A=Multiply(A=ComponentMask(None=@n27) {g=true}, B=TextureSample(UVs=@n31) {texture=&T_LightFalloffArray}.B)
   B=Multiply(A=Multiply(A=@n33, B=@n30.A), B=ComponentMask(None=@n27) {b=true})
  )
  n35 = Multiply(A=@n19, B=Add(A=Multiply(A=@n26.G, B=@n28), B=@n34))
  FunctionOutput(None=Clamp(None=Multiply(A=@n18, B=@n35))) {outputName=Custom_Lighting}
  n36 = Multiply(A=@n34, B=@n19)
  n37 = OneMinus(None=ComponentMask(None=@n2) {b=true})
  n38 = Multiply(B=OneMinus(None=Multiply(A=@n37)))
  n39 = NamedRerouteDeclaration(None=@n9) {name=MatType_ADD, nodeColor=<rgba 1 0 0.375 1>}
  n40 = TextureSample(UVs=ComponentMask(None=AppendVector(A=@n38, B=@n39)) {g=true, r=true}) {texture=&T_LightFalloffArray}
  FunctionOutput(None=Multiply(A=Add(B=@n36), B=Clamp(None=@n40.A))) {outputName=Custom_Lighting_Diffuse}
  FunctionOutput(None=@n22) {outputName=Tertiary_normal}
  LightVector()
  Multiply(A=@n3) {constB=0.5}
 ]}
