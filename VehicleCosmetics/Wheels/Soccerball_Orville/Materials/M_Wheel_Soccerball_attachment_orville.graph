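# Two-sided skeletal-mesh material for the Orville Soccerball wheel attachment, faking an animated gem with rainbow refraction and sheen.
# Fresnel terms from a panning normal map and chaos-animated RGB texture shade team or custom color blended with paint color.
# Thin bands cut from rotated, bump-offset UVs sample a scrolling gradient as fake refraction.
# These bands, a switchable moving sheen and iridescent rim feed color and intensity-scaled emissive math.
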Material /VehicleCosmetics/Wheels/Soccerball_Orville/Materials/M_Wheel_Soccerball_attachment_orville
graph {
 properties {bAutomaticallySetUsageInEditor=false, bUsedWithSkeletalMesh=true, twoSided=true, usedShadingModels=Default Lit}
 nodes [
  n1 = Constant3Vector() {constant=<rgba 0 0 1 0>}
  n2 = ScalarParameter() {defaultValue=1, parameterName=gem_fresnel_normal_intensity}
  n3 = Time()
  n4 = VectorParameter() {parameterName=RGBCHAOS}
  n5 = Multiply(A=@n3, B=@n4.B)
  n6 = VectorParameter() {parameterName=NormalTransform, defaultValue=<rgba 2 0.5 0 0>}
  n7 = Add(A=Multiply(A=TextureCoordinate(), B=AppendVector(A=@n6.R, B=@n6.R)) {constA=1}, B=AppendVector(A=@n6.B, B=@n6.B))
  n8 = Panner(Coordinate=@n7, Time=@n5) {speedY=1}
  n9 = TextureSampleParameter2D(UVs=@n8) {parameterName=NormalAO, samplerType=SAMPLERTYPE_Normal, texture=/VehicleCosmetics/Wheels/Soccerball_Orville/Textures/Orville_N}
  n10 = LinearInterpolate(A=Constant3Vector() {constant=<rgba 0 0 1 0>}, B=@n9.RGB, Alpha=Constant() {r=0.5})
  n11 = LinearInterpolate(A=@n1, B=@n10, Alpha=@n2)
  n12 = ConstantBiasScale(None=OneMinus(None=Fresnel(Normal=@n11) {exponent=2})) {bias=-0.5, scale=1.5}
  n13 = VectorParameter() {parameterName=TeamColor, defaultValue=<rgba 0.818 0 0.0018 1>}
  n14 = VectorParameter() {parameterName=CustomColor, defaultValue=<rgba 0 0.182 1 1>}
  n15 = StaticSwitchParameter(True=@n13.RGB, False=@n14.RGB) {parameterName=SwapTeam_CustomColor}
  n16 = Multiply(A=LinearInterpolate(A=@n15, B=VectorParameter() {parameterName=PaintColor, defaultValue=<rgba 0 1 0.0233 0.956>}.RGB))
  n17 = LinearInterpolate(B=@n16, Alpha=Power(Base=Clamp(None=@n12), Exp=ScalarParameter() {defaultValue=1, parameterName=Gem_Int_Fresnel}))
  n18 = Panner(Coordinate=Rotator(Coordinate=@n7, Time=Multiply(A=@n3, B=@n4.R)), Time=Multiply(A=@n3, B=@n4.G)) {speedX=1}
  n19 = TextureSampleParameter2D(UVs=BumpOffset(Coordinate=Panner(Coordinate=@n18, Time=@n5) {speedY=1}, Height=Constant(), HeightRatioInput=@n4.A)) {parameterName=RGB, samplerType=SAMPLERTYPE_LinearColor, texture=/VehicleCosmetics/Wheels/Soccerball_Orville/Textures/Orville_RGB}
  n20 = AppendVector(A=@n19.G, B=@n19.G)
  n21 = LinearInterpolate(A=AppendVector(A=@n20, B=Constant() {r=1}), B=@n10, Alpha=ScalarParameter() {defaultValue=0.5, parameterName=gme_ext_fresnel_messaround})
  n22 = Power(
   Base=Fresnel(Normal=LinearInterpolate(A=@n1, B=@n21, Alpha=@n2)) {exponent=2}
   Exp=ScalarParameter() {defaultValue=1, parameterName=Gem_ext_fresnel}
  )
  n23 = Clamp(None=StaticSwitchParameter(True=ConstantBiasScale(None=@n22) {bias=-0.02, scale=2}, False=ConstantBiasScale(None=@n22) {bias=0.02, scale=0.25}) {parameterName=GemExtDarkStyle})
  n24 = NamedRerouteUsage()
  n25 = TextureSampleParameter2D(UVs=Panner(Coordinate=ConstantBiasScale(None=@n24) {bias=0, scale=2}) {speedX=0.1}) {parameterName=Rainbo, samplerType=SAMPLERTYPE_LinearColor, texture=/VehicleCosmetics/Wheels/Soccerball_Orville/Textures/Orville_Gradient}
  n26 = ConstantBiasScale(None=NamedRerouteUsage()) {bias=-0.5, scale=0.2}
  n27 = Fresnel(Normal=@n11) {baseReflectFraction=0.1}
  n28 = ConstantBiasScale(None=@n27) {bias=-0.2, scale=1}
  n29 = Clamp(None=Multiply(A=Clamp(None=@n26), B=Clamp(None=@n28)))
  n30 = Multiply(A=NamedRerouteUsage(), B=Constant() {r=0.3})
  n31 = BumpOffset(Coordinate=TextureCoordinate(), Height=Constant() {r=0.4}, HeightRatioInput=Constant() {r=1})
  n32 = Multiply(A=Time(), B=Constant() {r=0.3})
  n33 = Rotator(Coordinate=Add(A=@n30, B=@n31), Time=@n32)
  n34 = ConstantBiasScale(None=ComponentMask(None=@n33) {r=true}) {bias=0, scale=1}
  n35 = Constant()
  n36 = TextureSampleParameter2D(UVs=Panner(Coordinate=AppendVector(A=@n34, B=@n35)) {speedX=0.2}) {parameterName=Rainbo, samplerType=SAMPLERTYPE_LinearColor, texture=&Orville_Gradient}
  n37 = NamedRerouteUsage()
  n38 = ComponentMask(None=@n33) {g=true}
  n39 = Subtract(A=Clamp(None=ConstantBiasScale(None=@n38) {bias=-0.4, scale=15}), B=Clamp(None=ConstantBiasScale(None=@n38) {bias=-0.425, scale=25}))
  n40 = Constant2Vector() {g=-0.2}
  n41 = Rotator(Coordinate=Add(A=Add(A=@n31, B=@n40), B=@n30), Time=@n32) {speed=-0.5}
  n42 = ComponentMask(None=@n41) {g=true}
  n43 = Subtract(A=Clamp(None=ConstantBiasScale(None=@n42) {bias=-0.375, scale=15}), B=Clamp(None=ConstantBiasScale(None=@n42) {bias=-0.475, scale=25}))
  n44 = Multiply(A=ConstantBiasScale(None=Add(A=@n39, B=@n43)) {bias=0, scale=2}, B=NamedRerouteUsage())
  n45 = Multiply(A=@n36.RGB, B=Clamp(None=Multiply(A=@n37, B=@n44)))
  n46 = Add(A=LinearInterpolate(A=LinearInterpolate(A=@n17, B=@n16, Alpha=@n23), B=@n25.RGB, Alpha=@n29), B=@n45)
  n47 = NamedRerouteUsage()
  n48 = NamedRerouteUsage()
  n49 = Constant() {r=0.25}
  n50 = Multiply(A=Clamp(None=@n39), B=Clamp(None=@n43))
  Add(
   A=LinearInterpolate(A=Add(A=@n46, B=@n47), B=Desaturation(None=@n48, Fraction=@n49), Alpha=ConstantBiasScale(None=@n50) {bias=0, scale=1})
   B=NamedRerouteUsage()
  )
  n51 = NamedRerouteUsage()
  n52 = Constant() {r=1}
  n53 = NamedRerouteUsage()
  n54 = ConstantBiasScale(None=Time()) {bias=0, scale=0.07}
  n55 = ConstantBiasScale(None=Abs(None=Sine(None=@n54))) {bias=0, scale=1.25}
  n56 = Multiply(A=NamedRerouteUsage(), B=ConstantBiasScale(None=Multiply(A=@n43, B=@n55)) {bias=-0.5, scale=3})
  n57 = ConstantBiasScale(None=ComponentMask(None=@n41) {r=true}) {bias=0, scale=1}
  n58 = Constant()
  n59 = TextureSampleParameter2D(UVs=Panner(Coordinate=AppendVector(A=@n57, B=@n58)) {speedX=0.25}) {parameterName=Rainbo, samplerType=SAMPLERTYPE_LinearColor, texture=&Orville_Gradient}
  n60 = Multiply(
   A=Clamp(None=Multiply(A=@n53, B=@n56))
   B=LinearInterpolate(A=NamedRerouteUsage(), B=@n59.RGB, Alpha=ScalarParameter() {defaultValue=0.5, parameterName=CenterGradientLerp})
  )
  n61 = NamedRerouteUsage()
  n62 = VectorParameter() {parameterName=CustomColor, defaultValue=<rgba 0 0.182 1 1>}
  n63 = TextureCoordinate()
  n64 = ConstantBiasScale(None=@n30) {bias=0, scale=0.29}
  n65 = Subtract(A=@n63, B=Multiply(A=@n64, B=Constant() {r=1}))
  n66 = TextureSampleParameter2D(UVs=Panner(Coordinate=ComponentMask(None=@n65) {r=true}) {speedX=0.2}) {parameterName=Rainbo, samplerType=SAMPLERTYPE_LinearColor, texture=&Orville_Gradient}
  n67 = ScalarParameter() {defaultValue=0.5, parameterName=CenterGradientLerp}
  n68 = Time()
  n69 = ScalarParameter() {defaultValue=5, parameterName=MainSheenRotateSpeed}
  n70 = Subtract(
   A=Rotator(Coordinate=@n63, Time=Multiply(A=@n68, B=@n69))
   B=Multiply(A=@n64, B=ScalarParameter() {defaultValue=0.7, parameterName=MainSheenRefraction})
  )
  n71 = Abs(None=ConstantBiasScale(None=ComponentMask(None=@n70) {r=true}) {bias=-0.46, scale=1})
  n72 = Constant() {r=0.95}
  n73 = Sine(None=ConstantBiasScale(None=@n68) {bias=0, scale=0.09})
  n74 = ConstantBiasScale(None=@n73) {bias=0, scale=0.012}
  n75 = Multiply(A=ScalarParameter() {defaultValue=10, parameterName=MainSheenMultiplier}, B=Subtract(A=OneMinus(None=@n71), B=Add(A=@n72, B=@n74)))
  n76 = ComponentMask(None=@n65) {r=true}
  n77 = Subtract(
   A=Abs(None=ConstantBiasScale(None=@n76) {bias=-0.49, scale=1})
   B=Add(A=Constant() {r=0.09}, B=ConstantBiasScale(None=@n73) {bias=0, scale=0.02})
  )
  n78 = Constant() {r=50}
  n79 = Clamp(None=Multiply(A=@n77, B=@n78))
  n80 = Constant() {r=0.015}
  n81 = Clamp(None=Multiply(A=@n78, B=Subtract(A=@n77, B=@n80)))
  n82 = Subtract(
   A=StaticSwitchParameter(True=Clamp(None=@n75), False=Subtract(A=@n79, B=@n81)) {parameterName=MainSheenVariation}
   B=ConstantBiasScale(None=NamedRerouteUsage()) {bias=0, scale=0.75}
  )
  n83 = Add(A=Add(A=@n45, B=Add(A=@n60, B=@n61)), B=Multiply(A=LinearInterpolate(A=@n62.RGB, B=@n66.RGB, Alpha=@n67), B=Clamp(None=@n82)))
  n84 = Multiply(A=@n25.RGB, B=@n29)
  n85 = Multiply(A=Multiply(A=StaticSwitchParameter(True=@n51, False=@n52) {parameterName=Param_1}, B=Add(A=@n83, B=@n84)))
  n86 = ScalarParameter() {defaultValue=1, parameterName=EmissiveIntensity}
  LinearInterpolate(A=Add(B=Multiply(A=@n85, B=@n86)), B=Constant())
  NamedRerouteDeclaration(None=TextureSample(UVs=@n8) {samplerType=SAMPLERTYPE_LinearColor, texture=/VehicleCosmetics/Wheels/Soccerball_Orville/Textures/Orville_alpha}.RGB) {name=Normal_alpha, nodeColor=<rgba 1 0.9375 0 1>}
  NamedRerouteDeclaration(None=@n20) {name=append, nodeColor=<rgba 0.6172 0 1 1>}
  NamedRerouteDeclaration(None=StaticSwitchParameter(True=@n14.RGB, False=@n13.RGB) {parameterName=SwapTeam_CustomColor}) {name=ST_CC, nodeColor=<rgba 0 0.3437 1 1>}
  n87 = ConstantBiasScale(None=@n19.B) {bias=-0.5, scale=0.1}
  NamedRerouteDeclaration(None=ConstantBiasScale(None=Multiply(A=@n27, B=@n87)) {bias=0.025, scale=3}) {name=CBS, nodeColor=<rgba 0.6641 1 0 1>}
  NamedRerouteDeclaration(None=@n15) {name=ST_CC_2, nodeColor=<rgba 0.2656 1 0 1>}
  NamedRerouteDeclaration(None=@n9.RGB) {name=Normal, nodeColor=<rgba 0 0.2969 1 1>}
  NamedRerouteDeclaration(None=@n60) {name=Refraction1, nodeColor=<rgba 1 0.5625 0 1>}
  n88 = Constant2Vector() {g=-0.4}
  n89 = Rotator(Coordinate=Add(A=Add(A=@n31, B=@n88), B=@n30), Time=@n32) {speed=-0.3}
  n90 = ComponentMask(None=@n89) {r=true}
  n91 = Constant()
  n92 = TextureSampleParameter2D(UVs=Panner(Coordinate=AppendVector(A=@n90, B=@n91)) {speedX=0.25}) {parameterName=Rainbo, samplerType=SAMPLERTYPE_LinearColor, texture=&Orville_Gradient}
  n93 = ComponentMask(None=@n89) {g=true}
  n94 = ConstantBiasScale(None=@n93) {bias=-0.425, scale=15}
  n95 = ConstantBiasScale(None=@n93) {bias=-0.45, scale=25}
  n96 = ConstantBiasScale(None=Subtract(A=Clamp(None=@n94), B=Clamp(None=@n95))) {bias=0, scale=2}
  n97 = Multiply(A=NamedRerouteUsage(), B=Multiply(A=NamedRerouteUsage(), B=ConstantBiasScale(None=@n96) {bias=0, scale=2}))
  NamedRerouteDeclaration(None=Multiply(A=@n92.RGB, B=Clamp(None=@n97))) {name=refraction2, nodeColor=<rgba 0.2422 0 1 1>}
  NamedRerouteUsage()
  ScalarParameter() {parameterName=Specular}
  ScalarParameter() {parameterName=Rouhness}
 ]
}
