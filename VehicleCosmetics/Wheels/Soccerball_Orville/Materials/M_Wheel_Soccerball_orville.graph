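# Masked wheel material for the Soccerball Orville vehicle cosmetic, usable on skeletal meshes.
# MF_Orville supplies emissive, base color and normal, which blend with MF_WheelParent_Simple output using an inverted rim-mask texture green channel and vertex color red.
# The merged material attributes then pass through vehicle damage and de-rez transition functions before output.
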Material /VehicleCosmetics/Wheels/Soccerball_Orville/Materials/M_Wheel_Soccerball_orville
graph {
 properties {
  bAutomaticallySetUsageInEditor=false
  bUseMaterialAttributes=true
  bUsedWithMeshDeformer=true
  bUsedWithSkeletalMesh=true
  blendMode=BLEND_Masked
  pixelDepthOffsetMode=PDOM_Legacy
  usedShadingModels=Default Lit
 }
 nodes [
  n1 = GetMaterialAttributes(None=MaterialFunctionCall() {materialFunction=&MF_WheelParent_Simple}.MaterialAttributes) {attributeGetTypes [#1, #2, #3]}
  n2 = NamedRerouteUsage()
  n3 = MaterialFunctionCall() {materialFunction=&/MF_Orville}
  n4 = NamedRerouteUsage()
  n5 = VertexColor()
  n6 = LinearInterpolate(A=LinearInterpolate(A=Subtract(A=@n3.Emissive_color, B=@n2), B=@n4, Alpha=@n2), B=@n4, Alpha=@n5.R) {constB=0}
  n7 = NamedRerouteUsage()
  n8 = NamedRerouteUsage()
  n9 = LinearInterpolate(A=LinearInterpolate(A=Subtract(A=@n3.Base_color, B=@n7), B=@n8, Alpha=@n7), B=@n8, Alpha=@n5.R) {constB=0}
  n10 = NamedRerouteUsage()
  n11 = NamedRerouteUsage()
  n12 = LinearInterpolate(A=LinearInterpolate(A=Subtract(A=@n3.Normal, B=@n10), B=@n11, Alpha=@n10), B=@n11, Alpha=VertexColor().R) {constB=0}
  MaterialFunctionCall(In=MaterialFunctionCall(Material=SetMaterialAttributes!(MaterialAttributes=@n1.MaterialAttributes, Emissive Color=@n6, Base Color=@n9, Normal=@n12) {attributeSetTypes [#2, #1, #3]}) {materialFunction=&MF_BaseVehicle_Damage}.Result) {materialFunction=&MF_VehicleDeRez_Transition}
  NamedRerouteDeclaration(None=@n1.Base Color) {name=BC, nodeColor=<rgba 1 0.7734 0 1>}
  NamedRerouteDeclaration(None=@n1.Emissive Color) {name=Emiss, nodeColor=<rgba 0.4531 0 1 1>}
  n13 = TextureSampleParameter2D() {
   parameterName=Rim_emissive_mask
   samplerType=SAMPLERTYPE_LinearColor
   texture=/VehicleCosmetics/Wheels/Soccerball_Orville/Textures/T_Soccerball_Orville_Rim_M
  }
  NamedRerouteDeclaration(None=OneMinus(None=@n13.G)) {name=Rim_mask, nodeColor=<rgba 0 1 0.1328 1>}
  NamedRerouteDeclaration(None=@n1.Normal) {name=Normal, nodeColor=<rgba 1 0 0.1875 1>}
 ]
}
